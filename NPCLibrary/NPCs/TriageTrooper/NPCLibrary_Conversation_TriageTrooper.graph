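# Conversation database for the Triage Trooper NPC, a medic specialist, defining its greeting and top-level conversation branches.
# The entry point leads to a Speech node with a default line and contextual messages gated by first-ever conversation, per-match conversation count, or specific player outfits.
# Priority curves rank the first-meeting and outfit reaction lines.
# The speech then branches through link nodes to the shared NPC library quests, basic services, and Triage Trooper dialogue entries.
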ConversationDatabase /NPCLibrary/NPCs/TriageTrooper/NPCLibrary_Conversation_TriageTrooper
entry_points [{tag=Conversation.Entry.TriageTrooper, destinations [#1]}]
nodes [
 ConversationEntryPointNode_7 {
  id=#1
  class=ConversationEntryPointNode
  outputs [#2]
  properties {entryTag {tagName=Conversation.Entry.TriageTrooper}}
 }
 Speech_C_4 {
  id=#2
  class=Speech_C
  outputs [#3, #4, #5]
  properties {
   nodeName=""
   generalConfig {
    defaultMessage=You hurt? I could fix that.
    contextualMessages [
     {
      message="Happy to lend a healing hand!"
      contextRequirements [{requirement {class=FortControllerRequirement_IsFirstEverConversationWithNPC}}]
      priority {curve {rowName=Tandem.Speech.Priority.FirstEverInteraction}}
     }
     {
      message=You hurt? I could fix that.
      contextRequirements [{requirement {
         class=FortControllerRequirement_CheckNPCMatchConversationCount
         properties {nPCUniqueIdTag {tagName=AISpawnerData.Type.Specialist.Medic.TriageTrooper}}
        }}]
      priority {curve {rowName=None}}
     }
     {
      message="I'm always here to help! Just ask Field Surgeon."
      contextRequirements [{requirement {
         class=FortControllerRequirement_CheckNPCMatchConversationCount
         properties {nPCUniqueIdTag {tagName=AISpawnerData.Type.Specialist.Medic.TriageTrooper}}
        }}]
      priority {curve {rowName=None}}
     }
     {
      message="If you're looking for some heals, I'm your man!"
      contextRequirements [{requirement {
         class=FortControllerRequirement_CheckNPCMatchConversationCount
         properties {nPCUniqueIdTag {tagName=AISpawnerData.Type.Specialist.Medic.TriageTrooper}}
        }}]
      priority {curve {rowName=None}}
     }
     {
      message=Tell Field Surgeon I get the message. I'm done running.
      contextRequirements [{requirement {class=FortControllerRequirement_HasCID, properties {allowedCIds [/BRCosmetics/Athena/Items/Cosmetics/Characters/Character_Bountress]}}}]
      priority {curve {rowName=Tandem.Speech.Priority.OutfitReaction}}
     }
     {
      message="I know we don't agree on what I did, but we can still be friends. Right...?"
      contextRequirements [{requirement {
         class=FortControllerRequirement_HasCID
         properties {allowedCIds [/BRCosmetics/Athena/Items/Cosmetics/Characters/CID_216_Athena_Commando_F_Medic]}
        }}]
      priority {curve {rowName=Tandem.Speech.Priority.OutfitReaction}}
     }
     {
      message="Zeus! It's... good to see you up and about."
      contextRequirements [{requirement {
         class=FortControllerRequirement_HasCID
         properties {allowedCIds [&Character_StunningMountain_Armor]}
        }}]
      priority {curve {rowName=Tandem.Speech.Priority.OutfitReaction}}
     }
    ]
   }
  }
 }
 ConversationLinkNode_2 {
  id=#5
  class=ConversationLinkNode
  properties {remoteEntryTag {tagName=Conversation.Entry.TriageTrooper.Dialogue}}
 }
 ConversationLinkNode_0 {id=#4, class=ConversationLinkNode, properties {remoteEntryTag {tagName=Conversation.Entry.BasicServices}}}
 ConversationLinkNode_1 {id=#3, class=ConversationLinkNode, properties {remoteEntryTag {tagName=Conversation.Entry.NPCLibrary.Quests}}}
]
