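# Conversation database for the Triage Trooper NPC's lore dialogue, opened through its Dialogue entry tag.
# An "Ask about..." speech node offers a Field Surgeon and Saskia topic, a Saving Zeus topic and a Back option.
# Each topic plays a linear chain of speech lines, and both converge on a shared return-to-start node.
# The Saving Zeus topic requires the player to use the StunningMountain_Armor character, checked by a controller requirement.
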ConversationDatabase /NPCLibrary/NPCs/TriageTrooper/NPCLibrary_Conversation_TriageTrooper_Dialogue
entry_points [{tag=Conversation.Entry.TriageTrooper.Dialogue, destinations [#1]}]
nodes [
 ConversationEntryPointNode_0 {
  id=#1
  class=ConversationEntryPointNode
  outputs [#2]
  properties {entryTag {tagName=Conversation.Entry.TriageTrooper.Dialogue}}
 }
 Speech_C_0 {
  id=#2
  class=Speech_C
  outputs [#3, #4, #5]
  properties {
   nodeName=""
   softTaskIcon=&T_UI_NPC_question
   generalConfig {defaultMessage=Ask me anything. I'm an open medkit.}
  }
  sub_nodes [ConversationChoiceNode_5 {class=ConversationChoiceNode, properties {defaultChoiceDisplayText=Ask about...}}]
 }
 Speech_C_2 {
  id=#5
  class=Speech_C
  outputs [#6]
  properties {nodeName="", softTaskIcon=&T_UI_NPC_question, generalConfig {defaultMessage="Listen, about what I did..."}}
  sub_nodes [
   ConversationChoiceNode_15 {class=ConversationChoiceNode, properties {defaultChoiceDisplayText=Saving Zeus}}
   FortConversationRequirement_ControllerRequirement_0 {
    class=FortConversationRequirement_ControllerRequirement
    properties {
     nodeName=Require: Uses Character
     requirement {
      class=FortControllerRequirement_HasCID
      properties {allowedCIds [&Character_StunningMountain_Armor]}
     }
    }
   }
  ]
 }
 Speech_C_3 {
  id=#6
  class=Speech_C
  outputs [#7]
  properties {
   nodeName=""
   softTaskIcon=&T_UI_NPC_question
   generalConfig {defaultMessage=I don't agree with you. Your whole vendetta against mortals.}
  }
 }
 Speech_C_4 {
  id=#7
  class=Speech_C
  outputs [#8]
  properties {
   nodeName=""
   softTaskIcon=&T_UI_NPC_question
   generalConfig {defaultMessage="It kind of goes against my personal interests. You know, as a mortal myself."}
  }
 }
 Speech_C_5 {
  id=#8
  class=Speech_C
  outputs [#9]
  properties {
   nodeName=""
   softTaskIcon=&T_UI_NPC_question
   generalConfig {defaultMessage=But you were hurt. I couldn't just NOT save you. No matter what Field Surgeon thinks.}
  }
 }
 ReturnToStart_C_23 {
  id=#9
  class=ReturnToStart_C
  properties {softTaskIcon=&T_UI_NPC_Back}
  sub_nodes [ConversationChoiceNode_32 {class=ConversationChoiceNode, properties {defaultChoiceDisplayText=BACK}}]
 }
 Back_C_0 {
  id=#4
  class=Back_C
  sub_nodes [ConversationChoiceNode_7 {class=ConversationChoiceNode, properties {defaultChoiceDisplayText=BACK}}]
 }
 Speech_C_7 {
  id=#3
  class=Speech_C
  outputs [#10]
  properties {
   nodeName=""
   softTaskIcon=&T_UI_NPC_question
   generalConfig {defaultMessage=Field Surgeon hired Saskia to find me.}
  }
  sub_nodes [ConversationChoiceNode_0 {class=ConversationChoiceNode, properties {defaultChoiceDisplayText=Field Surgeon and Saskia}}]
 }
 Speech_C_15 {
  id=#10
  class=Speech_C
  outputs [#11]
  properties {
   nodeName=""
   softTaskIcon=&T_UI_NPC_question
   generalConfig {defaultMessage=She's...not too happy about something I did. Someone I saved.}
  }
 }
 Speech_C_14 {
  id=#11
  class=Speech_C
  outputs [#12]
  properties {
   nodeName=""
   softTaskIcon=&T_UI_NPC_question
   generalConfig {defaultMessage=But even after everything he did? He was still hurt.}
  }
 }
 Speech_C_1 {
  id=#12
  class=Speech_C
  outputs [#9]
  properties {
   nodeName=""
   softTaskIcon=&T_UI_NPC_question
   generalConfig {defaultMessage="God or not, I help those who need me."}
  }
 }
]
